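# Editor utility widget for Foley sound data entry, editing one tag-to-sound pair in a SoundLibrary's Sounds map.
# Two SinglePropertyViews point at the widget itself to expose its CurrentTag and Sound variables.
# A tag change re-keys the entry by removing PreviousTag, cached on Construct and after each change, then adding CurrentTag.
# A sound change replaces the value under CurrentTag when it differs from the stored one.
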EditorUtilityWidgetBlueprint /Game/Sounds/Fort_Audio_Blueprints/Utilities/EditorUtilityWidgets/Foley/EUW_Foley_SoundDataEntry
graphs [EventGraph {nodes [
   "On Property Changed (TagView)"!() -> (OutputDelegate, then, PropertyName:Name=None) {is=title K2Node_ComponentBoundEvent}
   > n1 = Branch(Condition=MatchesTag!(
     TagOne:Gameplay Tag Structure=Get CurrentTag() -> (CurrentTag:Sound Library Tag Structure)
     TagTwo:Gameplay Tag Structure=Get PreviousTag() -> (PreviousTag:Sound Library Tag Structure)
     bExactMatch:bool=true
    ) -> (ReturnValue:bool))
   Remove!(
    execute=@n1.else
    TargetMap:Map of Gameplay Tag Structures to Sound Base Soft Object References=Get Sounds!(self:Sound Library Object Reference=Get SoundLibrary() -> (SoundLibrary:Sound Library Object Reference)) -> (Sounds:Map of Gameplay Tag Structures to Sound Base Soft Object References)
    Key:"Gameplay Tag Structure (by ref)"=Get PreviousTag()
   ) -> (then, ReturnValue:bool)
   > Add!(
    execute
    TargetMap:Map of Gameplay Tag Structures to Sound Base Soft Object References=Get Sounds(self=Get SoundLibrary())
    Key:"Gameplay Tag Structure (by ref)"=Get CurrentTag()
    Value:"Sound Base Soft Object Reference (by ref)"=Get Sound() -> (Sound:Sound Base Soft Object Reference)
   ) -> (then) {is=title K2Node_CallFunction}
   > Set PreviousTag(PreviousTag:Sound Library Tag Structure=Get CurrentTag()) -> (Output_Get:Sound Library Tag Structure)
   "On Property Changed (SoundsView)"!() -> (OutputDelegate, then, PropertyName:Name=None) {is=title K2Node_ComponentBoundEvent}
   n2 = Get Sounds(self=Get SoundLibrary())
   n3 = Get CurrentTag()
   > n4 = Branch(Condition="Equal (SoftObjectReference)"(
     A:"Soft Object Reference (by ref)"=Find(TargetMap:Map of Gameplay Tag Structures to Sound Base Soft Object References=@n2, Key:"Gameplay Tag Structure (by ref)"=@n3) -> (Value:Sound Base Soft Object Reference)
     B:"Soft Object Reference (by ref)"=Get Sound()
    ) -> (ReturnValue:bool=~) {is=title K2Node_PromotableOperator})
   n5 = Get CurrentTag()
   Remove(execute=@n4.else, TargetMap=Get Sounds(self=Get SoundLibrary()), Key=@n5)
   > Add(TargetMap=Get Sounds(self=Get SoundLibrary()), Key=@n5, Value=Get Sound())
   Event PreConstruct()
   > SetObject(
    self=[Get TagView() -> (TagView:Single Property View Object Reference), Get SoundsView() -> (SoundsView:Single Property View Object Reference)]
    NewObject=Self-Reference()
   )
   Event Construct()
   > Set PreviousTag(PreviousTag=Get CurrentTag())
  ]}]
variables [
 CurrentTag {type=SoundLibraryTag, default {tagName=None}}
 PreviousTag {type=SoundLibraryTag, default {tagName=None}}
 Sound {type=SoundBase, default=null}
 SoundLibrary {type=SoundLibrary, default=null}
]
widgets {
 info {rootWidgetClass=&Overlay, widgetCount=8}
 widgets [
  Overlay_141 {widgetClassPath=&Overlay, widget=Overlay_141}
  Border_1 {
   widgetClassPath=&Border
   parent=Overlay_141
   widget=Border_1
   properties {
    horizontalAlignment=HAlign_Left
    verticalAlignment=VAlign_Center
    background {
     bIsDynamicallyLoaded=false
     drawAs=Image
     imageType=NoImage
     mirroring=NoMirror
     resourceName=None
     resourceObject=null
     tiling=NoTile
     imageSize=<32 32>
     margin=<ltrb 0 0 0 0>
     outlineSettings {
      bUseBrushTransparency=false
      roundingType=HalfHeightRadius
      width=0
      color {colorUseRule=UseColor_Specified, specifiedColor=<rgba 0 0 0 0>}
      cornerRadii=<0 0 0 0>
     }
     tintColor {colorUseRule=UseColor_Specified, specifiedColor=<rgba 0.0208 0.0208 0.0208 1>}
     uVRegion {bIsValid=false, max=<0 0>, min=<0 0>}
    }
    padding=<ltrb 0 0 0 0>
   }
   slot {class=OverlaySlot}
  }
  Border_0 {
   widgetClassPath=&Border
   parent=Border_1
   widget=Border_0
   properties {background {
     bIsDynamicallyLoaded=false
     drawAs=Border
     imageType=NoImage
     mirroring=NoMirror
     resourceName=None
     resourceObject=null
     tiling=NoTile
     imageSize=<32 32>
     margin=<ltrb 1 1 1 1>
     outlineSettings {
      bUseBrushTransparency=false
      roundingType=HalfHeightRadius
      width=0
      color {colorUseRule=UseColor_Specified, specifiedColor=<rgba 0 0 0 0>}
      cornerRadii=<0 0 0 0>
     }
     tintColor {colorUseRule=UseColor_Specified, specifiedColor=<rgba 0 0 0 1>}
     uVRegion {bIsValid=false, max=<0 0>, min=<0 0>}
    }}
   slot {class=BorderSlot, properties {horizontalAlignment=HAlign_Left, verticalAlignment=VAlign_Center, padding=<ltrb 0 0 0 0>}}
  }
  HorizontalBox_128 {widgetClassPath=&HorizontalBox, parent=Border_0, widget=HorizontalBox_128, slot {class=BorderSlot}}
  SizeBox_1 {
   widgetClassPath=&SizeBox
   parent=HorizontalBox_128
   widget=SizeBox_1
   properties {bOverride_WidthOverride=true, widthOverride=350}
   slot {class=HorizontalBoxSlot, properties {padding=<ltrb 10 10 10 10>}}
  }
  VerticalBox_0 {widgetClassPath=&VerticalBox, parent=SizeBox_1, widget=VerticalBox_0, slot {class=SizeBoxSlot}}
  TagView {
   widgetClassPath=&SinglePropertyView
   parent=VerticalBox_0
   widget=TagView
   properties {nameOverride=Library Tag, propertyName=CurrentTag}
   slot {class=VerticalBoxSlot}
  }
  SoundsView {
   widgetClassPath=&SinglePropertyView
   parent=VerticalBox_0
   widget=SoundsView
   properties {propertyName=Sound}
   slot {class=VerticalBoxSlot}
  }
 ]
}
