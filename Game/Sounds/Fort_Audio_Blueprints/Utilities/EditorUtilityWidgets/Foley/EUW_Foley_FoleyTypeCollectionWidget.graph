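# Editor utility widget for foley authoring that creates or selects a Sound Library for one foley type and lists its sounds.
# When no library is set, the button creates an SL_Footstep_ asset named with FoleyName in the current Content Browser path, then saves and stores it.
# A property view bound to the widget's SoundLibrary variable rebuilds the scroll box on change, adding one EUW_Foley_SoundDataEntry per gameplay tag in the Sounds map.
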EditorUtilityWidgetBlueprint /Game/Sounds/Fort_Audio_Blueprints/Utilities/EditorUtilityWidgets/Foley/EUW_Foley_FoleyTypeCollectionWidget
graphs [
 EventGraph {nodes [
   "On Pressed (SoundLibraryButton)"() {is=title K2Node_ComponentBoundEvent}
   > n1 = Get!(execute:Exec) -> (then:Exec, else:Exec, SoundLibrary:Sound Library Object Reference)
   ClearChildren(execute=@n1.else, self=Get TestScrollBox() -> (TestScrollBox:Editor Utility Scroll Box Object Reference))
   > n2 = GetAssetTools()
   > n3 = CreateAsset(
    self=@n2
    AssetName=Append!(A=SL_Footstep_, B=Get FoleyName() -> (FoleyName:String), C:String) -> (ReturnValue)
    PackagePath=GetCurrentContentBrowserPath!() -> (OutPath:String, ReturnValue:bool)
    AssetClass=/Script/SoundLibrary
    CallingContext=Foley
   )
   > n4 = SaveLoadedAsset(AssetToSave=@n3, bOnlyIfIsDirty=true)
   > PrintString(InString=Append!(A=GetDisplayName(Object=@n3), B=SelectString(A=" Saved!", B=" Failed to Save!", bPickA=@n4)) -> (ReturnValue))
   > n5 = Cast To SoundLibrary!(execute, Object=@n3) -> (then, CastFailed, AsSound Library:Sound Library Object Reference) {is=title K2Node_DynamicCast}
   > Set SoundLibrary(SoundLibrary:Sound Library Object Reference=@n5) -> (Output_Get:Sound Library Object Reference)
   "On Property Changed (SoundLibraryView)"!() -> (OutputDelegate, then, PropertyName:Name=None) {is=title K2Node_ComponentBoundEvent}
   > ClearChildren(self=Get TestScrollBox())
   > Clear(TargetArray:Array of EUW Foley Sound Data Entry Object References=Get SoundDataEntryArray() -> (SoundDataEntryArray:Array of EUW Foley Sound Data Entry Object References))
   n6 = Get Sounds!(self:Sound Library Object Reference=Get SoundLibrary() -> (SoundLibrary:Sound Library Object Reference)) -> (Sounds:Map of Gameplay Tag Structures to Sound Base Soft Object References)
   > n7 = Keys!(execute, TargetMap:Map of Gameplay Tag Structures to Sound Base Soft Object References=@n6) -> (then, Keys:Array of Gameplay Tag Structures)
   > n8 = For Each Loop(Array:Array of Gameplay Tag Structures=@n7) -> (Array Element:Gameplay Tag Structure)
   > n9 = Create Widget!(
    execute:Exec
    WorldContextObject:Object Reference
    Class:User Widget Class Reference=&/EUW_Foley_SoundDataEntry.EUW_Foley_SoundDataEntry_C
    OwningPlayer:Player Controller Object Reference
    CurrentTag:Sound Library Tag Structure=MakeSoundLibraryTag!(Tag:"Gameplay Tag Structure (by ref)"=@n8) -> (ReturnValue:Sound Library Tag Structure)
    Sound:Sound Base Soft Object Reference=Find(TargetMap:Map of Gameplay Tag Structures to Sound Base Soft Object References=@n6, Key:"Gameplay Tag Structure (by ref)"=@n8) -> (Value:Sound Base Soft Object Reference)
    SoundLibrary:Sound Library Object Reference=Get SoundLibrary()
   ) -> (then:Exec, ReturnValue:EUW Foley Sound Data Entry Object Reference)
   > Add(
    TargetArray:Array of EUW Foley Sound Data Entry Object References=Get SoundDataEntryArray()
    NewItem:"EUW Foley Sound Data Entry Object Reference (by ref)"=@n9
   ) {is=title K2Node_CallArrayFunction}
   > AddChild(self=Get TestScrollBox(), Content=@n9)
   Event PreConstruct()
   > SetObject(self=Get SoundLibraryView() -> (SoundLibraryView:Single Property View Object Reference), NewObject=Self-Reference())
   Event Construct()
  ]}
 OnShoeTypeButton {nodes [OnShoeTypeButton!() -> (then, Widget:EUW Foley Foley Type Entry Widget Object Reference) {is=title K2Node_FunctionEntry}]}
]
variables [
 FoleyName {type=string, default=""}
 OnShoeTypeButton {default=()}
 ScrollBox {type=EditorUtilityScrollBox, default=null}
 SoundDataEntryArray {type="array<EUW_Foley_SoundDataEntry_C>", default []}
 SoundLibrary {type=SoundLibrary, default=null}
]
widgets {
 info {rootWidgetClass=&GridPanel, widgetCount=8}
 widgets [
  GridPanel_79 {widgetClassPath=&GridPanel, widget=GridPanel_79}
  SizeBox_0 {
   widgetClassPath=&SizeBox
   parent=GridPanel_79
   widget=SizeBox_0
   properties {bOverride_WidthOverride=true, widthOverride=250}
   slot {class=GridSlot, properties {row=2, padding=<ltrb 10 10 10 10>}}
  }
  SoundLibraryView {
   widgetClassPath=&SinglePropertyView
   parent=SizeBox_0
   widget=SoundLibraryView
   properties {nameOverride="SL: ", propertyName=SoundLibrary}
   slot {class=SizeBoxSlot}
  }
  FoleyNameText {
   widgetClassPath=/Script/Blutility.EditorUtilityEditableText
   bIsVariable=true
   parent=GridPanel_79
   widget=FoleyNameText
   properties {text=Example}
   slot {class=GridSlot, properties {padding=<ltrb 5 5 5 5>}}
  }
  SizeBox_3 {
   widgetClassPath=&SizeBox
   parent=GridPanel_79
   widget=SizeBox_3
   properties {bOverride_HeightOverride=true, bOverride_WidthOverride=true, heightOverride=350, minDesiredWidth=450, widthOverride=350}
   slot {class=GridSlot, properties {row=3}}
  }
  TestScrollBox {
   widgetClassPath=&EditorUtilityScrollBox
   bIsVariable=true
   parent=SizeBox_3
   widget=TestScrollBox
   slot {class=SizeBoxSlot, properties {padding=<ltrb 5 5 5 5>}}
  }
  SoundLibraryButton {
   widgetClassPath=&EditorUtilityButton
   parent=GridPanel_79
   widget=SoundLibraryButton
   slot {class=GridSlot, properties {horizontalAlignment=HAlign_Left, row=1, padding=<ltrb 5 5 5 5>}}
  }
  UEFN_TextBlock_C_4 {
   widgetClassPath=&UEFN_TextBlock
   parent=SoundLibraryButton
   widget=UEFN_TextBlock_C_4
   properties {
    justification=Center
    text=Create Sound Library
    font {
     bForceMonospaced=false
     bMaterialIsStencil=false
     fontMaterial=null
     fontObject=/LauncherSocial/BrutalType
     letterSpacing=0
     monospacedWidth=1
     size=9
     skewAmount=0
     typefaceFontName=Regular
     outlineSettings {
      bApplyOutlineToDropShadows=false
      bMiteredCorners=false
      bSeparateFillAlpha=false
      outlineMaterial=null
      outlineSize=0
      outlineColor=<rgba 0 0 0 1>
     }
    }
    shadowColorAndOpacity=<rgba 0 0 0 1>
   }
   slot {class=ButtonSlot}
  }
 ]
}
