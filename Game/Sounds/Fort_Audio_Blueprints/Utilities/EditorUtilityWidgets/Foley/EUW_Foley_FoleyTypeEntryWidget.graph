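# Editor utility widget for one foley type entry row in the Foley Wizard Tool, showing a text label beside an X button.
# Pressing X finds the wizard by casting the outer of the parent's outer, then removes collection widgets with a matching FoleyName from its ListGridPanel.
# The entry then removes itself from its parent and calls UpdateGridSlots on the wizard.
# An OnFoleyTypeRemoved dispatcher is declared but not called here.
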EditorUtilityWidgetBlueprint /Game/Sounds/Fort_Audio_Blueprints/Utilities/EditorUtilityWidgets/Foley/EUW_Foley_FoleyTypeEntryWidget
graphs [
 EventGraph {nodes [
   "On Pressed (FoleyTypeButton)"() {is=title K2Node_ComponentBoundEvent}
   n1 = GetParent!(self:Widget Object Reference) -> (ReturnValue:Panel Widget Object Reference)
   > n2 = Cast To EUW_Foley_WizardTool!(execute, Object=GetOuterObject!(Object:Object Reference=GetOuterObject(Object=@n1)) -> (ReturnValue:Object Reference)) -> (then, CastFailed, AsEUW Foley Wizard Tool:EUW Foley Wizard Tool Object Reference) {is=title K2Node_DynamicCast}
   n3 = Get ListGridPanel!(self:EUW Foley Wizard Tool Object Reference=@n2) -> (ListGridPanel:Uniform Grid Panel Object Reference)
   > n4 = For Each Loop(Array:Array of Widget Object References=GetAllChildren(self=@n3)) -> (Array Element:Widget Object Reference)
   RemoveChild(execute=@n4.Completed, self=GetParent(), Content=Self-Reference())
   > UpdateGridSlots!(execute, self:EUW Foley Wizard Tool Object Reference=@n2, Content:Widget Object Reference) -> (then)
   n5 = Cast To EUW_Foley_FoleyTypeCollectionWidget!(execute=@n4.LoopBody, Object=@n4) -> (then, CastFailed, AsEUW Foley Foley Type Collection Widget:EUW Foley Foley Type Collection Widget Object Reference) {is=title K2Node_DynamicCast}
   > Branch(Condition="Equal Exactly (String)"(A=Get FoleyName!(self:EUW Foley Foley Type Collection Widget Object Reference=@n5) -> (FoleyName:String), B=Get FoleyName!() -> (FoleyName)))
   > RemoveChild(self=@n3, Content=@n5)
  ]}
 OnFoleyTypeRemoved {nodes [OnFoleyTypeRemoved!() -> (then, FoleyTypeEntry:EUW Foley Foley Type Entry Widget Object Reference) {is=title K2Node_FunctionEntry}]}
]
variables [
 FoleyName {type=string, default=""}
 OnFoleyTypeRemoved {default=()}
]
widgets {
 info {rootWidgetClass=&Overlay, widgetCount=8}
 widgets [
  Overlay_141 {widgetClassPath=&Overlay, widget=Overlay_141}
  Border_1 {
   widgetClassPath=&Border
   parent=Overlay_141
   widget=Border_1
   properties {
    verticalAlignment=VAlign_Center
    background {
     bIsDynamicallyLoaded=false
     drawAs=Image
     imageType=NoImage
     mirroring=NoMirror
     resourceName=None
     resourceObject=null
     tiling=NoTile
     imageSize=<32 32>
     margin=<ltrb 0 0 0 0>
     outlineSettings {
      bUseBrushTransparency=false
      roundingType=HalfHeightRadius
      width=0
      color {colorUseRule=UseColor_Specified, specifiedColor=<rgba 0 0 0 0>}
      cornerRadii=<0 0 0 0>
     }
     tintColor {colorUseRule=UseColor_Specified, specifiedColor=<rgba 0.0208 0.0208 0.0208 1>}
     uVRegion {bIsValid=false, max=<0 0>, min=<0 0>}
    }
    padding=<ltrb 0 0 0 0>
   }
   slot {class=OverlaySlot, properties {horizontalAlignment=HAlign_Fill, verticalAlignment=VAlign_Center}}
  }
  Border_0 {
   widgetClassPath=&Border
   parent=Border_1
   widget=Border_0
   properties {background {
     bIsDynamicallyLoaded=false
     drawAs=Border
     imageType=NoImage
     mirroring=NoMirror
     resourceName=None
     resourceObject=null
     tiling=NoTile
     imageSize=<32 32>
     margin=<ltrb 1 1 1 1>
     outlineSettings {
      bUseBrushTransparency=false
      roundingType=HalfHeightRadius
      width=0
      color {colorUseRule=UseColor_Specified, specifiedColor=<rgba 0 0 0 0>}
      cornerRadii=<0 0 0 0>
     }
     tintColor {colorUseRule=UseColor_Specified, specifiedColor=<rgba 0 0 0 1>}
     uVRegion {bIsValid=false, max=<0 0>, min=<0 0>}
    }}
   slot {class=BorderSlot, properties {verticalAlignment=VAlign_Center, padding=<ltrb 0 0 0 0>}}
  }
  HorizontalBox_128 {widgetClassPath=&HorizontalBox, parent=Border_0, widget=HorizontalBox_128, slot {class=BorderSlot}}
  FoleyTypeButton {
   widgetClassPath=&EditorUtilityButton
   parent=HorizontalBox_128
   widget=FoleyTypeButton
   properties {widgetStyle {
     clickedSlateSound {resourceObject=null}
     disabled {
      bIsDynamicallyLoaded=false
      drawAs=RoundedBox
      imageType=NoImage
      mirroring=NoMirror
      resourceName=None
      resourceObject=null
      tiling=NoTile
      imageSize=<0 0>
      margin=<ltrb 0 0 0 0>
      outlineSettings {
       bUseBrushTransparency=false
       roundingType=FixedRadius
       width=1
       color {colorUseRule=UseColor_ColorTable, specifiedColor=<rgba 0 0 0 0>}
       cornerRadii=<4 4 4 4>
      }
      tintColor {colorUseRule=UseColor_Specified, specifiedColor=<rgba 0.0395 0.0395 0.0395 1>}
      uVRegion {bIsValid=false, max=<0 0>, min=<0 0>}
     }
     disabledForeground {colorUseRule=UseColor_ColorTable, specifiedColor=<rgba 0 0 0 0>}
     hovered {
      bIsDynamicallyLoaded=false
      drawAs=RoundedBox
      imageType=NoImage
      mirroring=NoMirror
      resourceName=None
      resourceObject=null
      tiling=NoTile
      imageSize=<32 32>
      margin=<ltrb 0 0 0 0>
      outlineSettings {
       bUseBrushTransparency=false
       roundingType=FixedRadius
       width=1
       color {colorUseRule=UseColor_ColorTable, specifiedColor=<rgba 0 0 0 0>}
       cornerRadii=<4 4 4 4>
      }
      tintColor {colorUseRule=UseColor_Specified, specifiedColor=<rgba 0.0953 0.0953 0.0953 1>}
      uVRegion {bIsValid=false, max=<0 0>, min=<0 0>}
     }
     hoveredForeground {colorUseRule=UseColor_ColorTable, specifiedColor=<rgba 0 0 0 0>}
     hoveredSlateSound {resourceObject=null}
     normal {
      bIsDynamicallyLoaded=false
      drawAs=RoundedBox
      imageType=NoImage
      mirroring=NoMirror
      resourceName=None
      resourceObject=null
      tiling=NoTile
      imageSize=<32 32>
      margin=<ltrb 0 0 0 0>
      outlineSettings {
       bUseBrushTransparency=false
       roundingType=FixedRadius
       width=1
       color {colorUseRule=UseColor_ColorTable, specifiedColor=<rgba 0 0 0 0>}
       cornerRadii=<4 4 4 4>
      }
      tintColor {colorUseRule=UseColor_Specified, specifiedColor=<rgba 0.0395 0.0395 0.0395 1>}
      uVRegion {bIsValid=false, max=<0 0>, min=<0 0>}
     }
     normalForeground {colorUseRule=UseColor_ColorTable, specifiedColor=<rgba 0 0 0 0>}
     normalPadding=<ltrb 12 1.5 12 1.5>
     pressed {
      bIsDynamicallyLoaded=false
      drawAs=RoundedBox
      imageType=NoImage
      mirroring=NoMirror
      resourceName=None
      resourceObject=null
      tiling=NoTile
      imageSize=<32 32>
      margin=<ltrb 0 0 0 0>
      outlineSettings {
       bUseBrushTransparency=false
       roundingType=FixedRadius
       width=1
       color {colorUseRule=UseColor_ColorTable, specifiedColor=<rgba 0 0 0 0>}
       cornerRadii=<4 4 4 4>
      }
      tintColor {colorUseRule=UseColor_Specified, specifiedColor=<rgba 0.0284 0.0284 0.0284 1>}
      uVRegion {bIsValid=false, max=<0 0>, min=<0 0>}
     }
     pressedForeground {colorUseRule=UseColor_ColorTable, specifiedColor=<rgba 0 0 0 0>}
     pressedPadding=<ltrb 12 2.5 12 0.5>
     pressedSlateSound {resourceObject=null}
    }}
   slot {class=HorizontalBoxSlot}
  }
  SizeBox_0 {
   widgetClassPath=&SizeBox
   parent=FoleyTypeButton
   widget=SizeBox_0
   properties {bOverride_WidthOverride=true, widthOverride=25}
   slot {class=ButtonSlot, properties {padding=<ltrb 0 0 0 0>}}
  }
  TypeTextBlock {
   widgetClassPath=&TextBlock
   bIsVariable=true
   parent=SizeBox_0
   widget=TypeTextBlock
   properties {
    text=X
    colorAndOpacity {colorUseRule=UseColor_Specified, specifiedColor=<rgba 1 0.0574 0 1>}
    font {
     bForceMonospaced=false
     bMaterialIsStencil=false
     fontMaterial=null
     fontObject=/LauncherSocial/BrutalType
     letterSpacing=0
     monospacedWidth=1
     size=15
     skewAmount=0
     typefaceFontName=Regular
     outlineSettings {
      bApplyOutlineToDropShadows=false
      bMiteredCorners=false
      bSeparateFillAlpha=false
      outlineMaterial=null
      outlineSize=0
      outlineColor=<rgba 0 0 0 1>
     }
    }
    shadowColorAndOpacity=<rgba 0.2042 0.54 0.8333 1>
    shadowOffset=<0.5 0.5>
   }
   slot {class=SizeBoxSlot, properties {horizontalAlignment=HAlign_Center, verticalAlignment=VAlign_Center}}
  }
  TextEntry {
   widgetClassPath=&TextBlock
   bIsVariable=true
   parent=HorizontalBox_128
   widget=TextEntry
   properties {
    text=Example
    colorAndOpacity {colorUseRule=UseColor_Specified, specifiedColor=<rgba 1 1 1 0.007>}
    font {
     bForceMonospaced=false
     bMaterialIsStencil=false
     fontMaterial=null
     fontObject=&BrutalType
     letterSpacing=0
     monospacedWidth=1
     size=15
     skewAmount=0
     typefaceFontName=Regular
     outlineSettings {
      bApplyOutlineToDropShadows=false
      bMiteredCorners=false
      bSeparateFillAlpha=false
      outlineMaterial=null
      outlineSize=0
      outlineColor=<rgba 0 0 0 1>
     }
    }
    shadowColorAndOpacity=<rgba 0.2042 0.54 0.8333 1>
    shadowOffset=<0.5 0.5>
   }
   slot {class=HorizontalBoxSlot, properties {padding=<ltrb 4 2 4 2>, size {sizeRule=Fill, value=1}}}
  }
 ]
}
