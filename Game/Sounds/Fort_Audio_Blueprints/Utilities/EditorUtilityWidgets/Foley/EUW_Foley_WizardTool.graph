EditorUtilityWidgetBlueprint /Game/Sounds/Fort_Audio_Blueprints/Utilities/EditorUtilityWidgets/Foley/EUW_Foley_WizardTool
graphs [
 AddFoleyCollection {nodes [
   n1 = AddFoleyCollection!() -> (then, FoleyName:String) {is=title K2Node_FunctionEntry}
   > n2 = Construct Object from Class!(
    execute
    Class=&/EUW_Foley_FoleyTypeCollectionWidget.EUW_Foley_FoleyTypeCollectionWidget_C
    self=Get ListGridPanel() -> (ListGridPanel:Uniform Grid Panel Object Reference)
    FoleyName:String=@n1
   ) -> (then, ReturnValue:EUW Foley Foley Type Collection Widget Object Reference)
   > UpdateGridSlots!(execute, self, Content:Widget Object Reference=@n2) -> (then)
  ]}
 CheckIfFoleyNameExists {nodes [
   n1 = CheckIfFoleyNameExists!() -> (then, FoleyName:String) {is=title K2Node_FunctionEntry}
   > Set ShoeNameExistsAlready(ShoeNameExistsAlready:bool=false) -> (Output_Get:bool)
   > n2 = For Each Loop(Array:Array of Widget Object References=GetAllChildren(self=Get FoleyTypeScrollBox() -> (FoleyTypeScrollBox:Editor Utility Scroll Box Object Reference))) -> (Array Element:Widget Object Reference)
   Return Node!(execute=@n2.Completed, Exists:bool=Get ShoeNameExistsAlready() -> (ShoeNameExistsAlready:bool))
   > n3 = Cast To EUW_Foley_FoleyTypeEntryWidget!(execute, Object=@n2) -> (then, CastFailed, AsEUW Foley Foley Type Entry Widget:EUW Foley Foley Type Entry Widget Object Reference) {is=title K2Node_DynamicCast}
   > Branch(Condition="Equal Exactly (String)"(A=Get FoleyName!(self:EUW Foley Foley Type Entry Widget Object Reference=@n3) -> (FoleyName:String), B=@n1))
   > Set ShoeNameExistsAlready(ShoeNameExistsAlready=true)
  ]}
 CheckIfWave {nodes [
   n1 = Inputs!() -> (Exec:Exec, WavesArray:Array of Object References)
   > Set HasNonWave(HasNonWave:bool=false) -> (Output_Get:bool)
   > n2 = For Each Loop(Array:Array of Object References=@n1) -> (Array Element:Object Reference)
   > n3 = Cast To SoundWave!(execute, Object=@n2) -> (then, CastFailed, AsSound Wave:Sound Wave Object Reference) {is=title K2Node_DynamicCast}
   Set HasNonWave(execute=@n3.CastFailed, HasNonWave=true)
   n4 = Branch(execute=@n2.Completed, Condition=NOT Boolean(A=Get HasNonWave() -> (HasNonWave:bool)))
   > Outputs!(AllWaves:Exec, HasNonWave:Exec=@n5.then)
   n5 = Show Message Dialog(
    execute=@n4.else
    Title=MakeLiteralText(Value="LOCTEXT(\"Has Non Wave Asset Selected\")")
    Message=MakeLiteralText(Value="LOCTEXT(\"Please check your selection. Only sound wave Asset are allow.\")")
    MessageType=Ok
    DefaultValue=No
    MessageCategory=Warning
   )
  ]}
 CompareObjects {nodes [
   CompareObjects() {is=title K2Node_FunctionEntry}
   > n1 = For Each Loop(Array=Get SelectedObject() -> (SelectedObject:Array of Object References))
   n2 = For Each Loop(Exec=@n1.Completed, Array=Get GroupedObjects() -> (GroupedObjects:Array of Object References))
   Return Node!(execute=@n2.Completed)
   > Remove Item!(execute, TargetArray:Array of Object References=Get SelectedObject(), Item:"Object Reference (by ref)"=@n2) -> (then, ReturnValue:bool)
   n3 = Get WaveName() -> (WaveName:String)
   n4 = GetObjectName(Object=@n1)
   n5 = Split(SourceString=SelectString(A=@n4, B=@n3, bPickA=IsEmpty(InString=@n3)), InStr=_, SearchDir=FromEnd)
   n6 = Split(SourceString=@n4, InStr=_, SearchDir=FromEnd)
   n7 = "Equal, Case Insensitive (String)"(A=@n5, B=@n6)
   Branch(execute=@n1.LoopBody, Condition=@n7)
   > Set WaveName(WaveName:String=@n4) -> (Output_Get:String)
   > Add(TargetArray:Array of Object References=Get GroupedObjects(), NewItem:"Object Reference (by ref)"=@n1) {is=title K2Node_CallArrayFunction}
   AND Boolean!(A=@n7, B=IsNumeric!(SourceString:String=@n5.RightS) -> (ReturnValue:bool), C:bool=IsNumeric(SourceString=@n6.RightS)) -> (ReturnValue)
  ]}
 CreateFoleyEntryWidget {nodes [
   n1 = CreateFoleyEntryWidget!() -> (then, FoleyName:String) {is=title K2Node_FunctionEntry}
   > n2 = Construct Object from Class(
    Class=&/EUW_Foley_FoleyTypeEntryWidget.EUW_Foley_FoleyTypeEntryWidget_C
    self=Get FoleyTypeScrollBox()
    FoleyName=@n1
   ) -> (ReturnValue:EUW Foley Foley Type Entry Widget Object Reference)
   > AddChild(self=Get FoleyTypeScrollBox(), Content=@n2)
  ]}
 EventGraph {nodes [
   "On Pressed (BuildTemplatesButton)"() {is=title K2Node_ComponentBoundEvent}
   > n1 = GetSelectedAssets()
   > n2 = Set SelectedObject(SelectedObject:Array of Object References=@n1) -> (Output_Get:Array of Object References)
   > Check if Wave!(Exec:Exec, WavesArray:Array of Object References=@n2) -> (AllWaves:Exec, HasNonWave:Exec) {is=title K2Node_MacroInstance}
   > CheckSelectedWaves()
   "On Pressed (DefaultFoleyTypeButton)"() {is=title K2Node_ComponentBoundEvent}
   > n3 = For Each Loop(Array:Array of Strings=Get DefaultFoleyTypes() -> (DefaultFoleyTypes:Array of Strings)) -> (Array Element:String)
   > n4 = CheckIfFoleyNameExists!(execute, self, FoleyName:String=@n3) -> (then, Exists:bool) {is=title K2Node_CallFunction}
   > n5 = Branch(Condition=@n4)
   CreateFoleyEntryWidget!(execute=@n5.else, self, FoleyName:String=@n3) -> (then) {is=title K2Node_CallFunction}
   > AddFoleyCollection!(execute, self, FoleyName:String=@n3) -> (then) {is=title K2Node_CallFunction}
   "On Pressed (AddToLibButton)"() {is=title K2Node_ComponentBoundEvent}
   > n6 = GetSelectedAssets()
   > n7 = For Each Loop(Array:Array of Object References=@n6) -> (Array Element:Object Reference)
   > n8 = Cast To SoundCueTemplate!(execute, Object=@n7) -> (then, CastFailed, AsSound Cue Template:Sound Cue Template Object Reference) {is=title K2Node_DynamicCast}
   > n9 = For Each Loop(Array:Array of Widget Object References=GetAllChildren(self=Get ListGridPanel())) -> (Array Element:Widget Object Reference)
   > n10 = Cast To EUW_Foley_FoleyTypeCollectionWidget!(execute, Object=@n9) -> (then, CastFailed, AsEUW Foley Foley Type Collection Widget:EUW Foley Foley Type Collection Widget Object Reference) {is=title K2Node_DynamicCast}
   n11 = GetDisplayName(Object=@n8)
   > Branch(Condition=Contains(SearchIn=@n11, Substring=Get FoleyName(self:EUW Foley Foley Type Collection Widget Object Reference=@n10)))
   > n12 = For Each Loop(Array:Array of Strings=Get SurfaceTypes() -> (SurfaceTypes:Array of Strings)) -> (Array Element:String)
   n13 = For Each Loop(Exec=@n12.Completed, Array=Get StanceType() -> (StanceType:Array of Strings))
   n14 = Append!(
    A=SoundLibrary.Foley.
    B=Get Array Element_0() -> (Array Element_0:String)
    C:String=.
    D:String=Get Array Element() -> (Array Element:String)
   ) -> (ReturnValue)
   n15 = RequestGameplayTag!(execute=@n13.Completed, TagName:String=@n14, bErrorIfNotFound:bool=true) -> (then, OutTag:Gameplay Tag Structure, ReturnValue:bool)
   > n16 = Branch(Condition=@n15.ReturnValue)
   > Add!(
    execute
    TargetMap:Map of Gameplay Tag Structures to Sound Base Soft Object References=Get Sounds!(self:Sound Library Object Reference=Get SoundLibrary!(self:EUW Foley Foley Type Collection Widget Object Reference=@n10) -> (SoundLibrary:Sound Library Object Reference)) -> (Sounds:Map of Gameplay Tag Structures to Sound Base Soft Object References)
    Key:"Gameplay Tag Structure (by ref)"=MakeSoundLibraryTag!(Tag:"Gameplay Tag Structure (by ref)"=@n15) -> (ReturnValue:Sound Library Tag Structure)
    Value:"Sound Base Soft Object Reference (by ref)"=Make Soft Reference!(Input:Sound Cue Template Object Reference=@n8) -> (Output:Sound Cue Template Soft Object Reference)
   ) -> (then) {is=title K2Node_CallFunction}
   n17 = Append!(A="This tag does not exist:\r\n", B=@n14) -> (ReturnValue)
   Show Message Dialog(
    execute=@n16.else
    Title=MakeLiteralText(Value="LOCTEXT(\"Missing Tag\")")
    Message=MakeLiteralText(Value="To Text (String)"(InString=@n17))
    MessageType=Ok
    DefaultValue=No
    MessageCategory=Warning
   )
   Branch(execute=@n13.LoopBody, Condition=Contains(SearchIn=@n11, Substring=@n13))
   > Set Array Element_0(Array Element_0:String=@n13) -> (Output_Get:String)
   Branch(execute=@n12.LoopBody, Condition=Contains(SearchIn=@n11, Substring=@n12))
   > Set Array Element(Array Element:String=@n12) -> (Output_Get:String)
   "On Pressed (AddFoleyTypeButton)"() {is=title K2Node_ComponentBoundEvent}
   n18 = Get FoleyTextEntry() -> (FoleyTextEntry:Editor Utility Editable Text Box Object Reference)
   n19 = Replace(SourceString="To String (Text)"(InText="GetText (Text Box)"!(self:Text Box Object Reference=@n18) -> (ReturnValue:Text)), From=" ")
   > n20 = Branch(Condition=IsEmpty(InString=@n19))
   n21 = CheckIfFoleyNameExists(execute=@n20.else, FoleyName=@n19)
   > n22 = Branch(Condition=@n21)
   CreateFoleyEntryWidget(execute=@n22.else, FoleyName=@n19)
   > AddFoleyCollection(FoleyName=@n19)
   CheckSelectedWaves() {is=title K2Node_CustomEvent}
   > Branch(Condition=Is Not Empty!(TargetArray:Array of Object References=Get SelectedObject()) -> (ReturnValue:bool))
   > Clear(TargetArray:Array of Object References=Get GroupedObjects())
   > CompareObjects() {is=title K2Node_CallFunction}
   > MakeOrFindTemplate()
   > CheckSelectedWaves() {is=title K2Node_CallFunction}
   n23 = For Each Loop(Array:Array of Object References) -> (Array Element:Object Reference)
   > Cast To SoundBase!(execute, Object=@n23) -> (then, CastFailed, AsSound Base:Sound Base Object Reference) {is=title K2Node_DynamicCast}
   Add(TargetMap=Get Sounds(), Value=Resolve Soft Reference!(Input:Wildcard) -> (Output:Wildcard))
   UpdateSoundLibraryWidgets()
   "Get (a copy)"(Array:Array of Object References=@n6) -> (Output:Object Reference)
   Reroute Node(InputPin:Wildcard) -> (OutputPin:Wildcard)
  ]}
 MakeGroupedFolder {nodes [
   n1 = MakeGroupedFolder!() -> (then, Folder:String) {is=title K2Node_FunctionEntry}
   n2 = Get Object Path String(Object="Get (a copy)"(Array=Get GroupedObjects()))
   > Return Node!(execute, ReturnValue:String=Append!(A=Split(SourceString=@n2, InStr=/Waves, SearchDir=FromEnd), B=/, C:String=@n1) -> (ReturnValue))
  ]}
 MakeNewAsset {nodes [
   n1 = MakeNewAsset!() -> (then, AssetName:String, PackagePath:String) {is=title K2Node_FunctionEntry}
   > n2 = GetAssetTools()
   > n3 = DuplicateAsset!(
    execute
    self:Asset Tools Interface=@n2
    AssetName:String=@n1
    PackagePath:String=@n1.PackagePath
    OriginalObject:Object Reference=&/Foley_FS_Template
   ) -> (then, ReturnValue:Object Reference)
   > n4 = SaveLoadedAsset(AssetToSave=@n3, bOnlyIfIsDirty=false)
   > PrintString(InString=Append!(A=GetDisplayName(Object=@n3), B=SelectString(A=" Saved!", B=" Failed to Save!", bPickA=@n4)) -> (ReturnValue))
   > Return Node!(execute, SoundCueTemplate:Object Reference=@n3)
   CreateAsset(self=@n2, AssetName=@n1, PackagePath=@n1.PackagePath, AssetClass=/Script/FortSoundCueTemplates.FootstepFoley, CallingContext=Foley)
  ]}
 MakeOrFindTemplate {nodes [
   MakeOrFindTemplate() {is=title K2Node_FunctionEntry}
   n1 = MakeGroupedFolder!(Folder:String=Cues) -> (ReturnValue:String) {is=title K2Node_CallFunction}
   > MakeDirectory!(
    execute
    self:Editor Asset Subsystem Object Reference=Get EditorAssetSubsystem() -> (ReturnValue:Editor Asset Subsystem Object Reference) {is=title K2Node_GetEditorSubsystem}
    DirectoryPath:String=@n1
   ) -> (then, ReturnValue:bool)
   > n2 = DoesAssetExist!(execute, self:Editor Asset Subsystem Object Reference=Get EditorAssetSubsystem(), AssetPath=@n1) -> (then, ReturnValue)
   > n3 = Branch(Condition=@n2)
   n4 = For Each Loop(Exec=@n3.else, Array:Array of Strings=Get DefaultFoleyTypes()) -> (Array Element:String)
   n5 = For Each Loop(Exec=@n4.Completed, Array=Get SurfaceTypes())
   n6 = For Each Loop(Exec=@n5.Completed, Array=Get StanceType())
   n7 = Get StanceName() -> (StanceName:String)
   n8 = IsEmpty(InString=@n7)
   n9 = Get FoleyName!() -> (FoleyName)
   n10 = IsEmpty(InString=@n9)
   n11 = Get SurfaceName() -> (SurfaceName:String)
   n12 = IsEmpty(InString=@n11)
   n13 = MakeNewAsset!(
    execute=@n6.Completed
    self
    AssetName:String=Append!(
     A=SC_Foley
     B=Select(Option 0:String=_, Option 1:String, Index=@n10) -> (ReturnValue:String)
     C:String=@n9
     D:String=Select(Option 0=_, Index=@n12)
     E:String=@n11
     F:String=Select(Option 0=_, Index=@n8)
     G:String=@n7
    ) -> (ReturnValue)
    PackagePath:String=@n1
   ) -> (then, SoundCueTemplate:Object Reference) {is=title K2Node_CallFunction}
   > n14 = Cast To SoundCueTemplate(Object=@n13)
   > n15 = SetEditorProperty(Object=@n14, PropertyName=WaveVariations, PropertyValue:Array of Object References=Get GroupedObjects())
   > PrintString(InString="To String (Boolean)"(InBool=@n15))
   n16 = GetDisplayName(Object="Get (a copy)"(Array=Get GroupedObjects()))
   Branch(execute=@n6.LoopBody, Condition=Contains(SearchIn=@n16, Substring=@n6))
   > Set StanceName(StanceName:String=@n6) -> (Output_Get:String)
   Branch(execute=@n5.LoopBody, Condition=Contains(SearchIn=@n16, Substring=@n5))
   > Set SurfaceName(SurfaceName:String=@n5) -> (Output_Get:String)
   Branch(execute=@n4.LoopBody, Condition=Contains(SearchIn=@n16, Substring=@n4))
   > Set FoleyName(FoleyName:String=@n4) -> (Output_Get:String)
  ]}
 UpdateGridSlots {nodes [
   n1 = UpdateGridSlots!() -> (then, Content:Widget Object Reference) {is=title K2Node_FunctionEntry}
   > Set GridColumn(GridColumn:int=0) -> (Output_Get:int=0)
   > Set GridRow(GridRow:int=0) -> (Output_Get:int=0)
   n2 = Get ListGridPanel()
   > AddChild(self=@n2, Content=@n1)
   > n3 = For Each Loop(Array:Array of Widget Object References=GetAllChildren(self=@n2)) -> (Array Element:Widget Object Reference)
   n4 = SlotAsUniformGridSlot!(Widget:Widget Object Reference=@n3) -> (ReturnValue:Uniform Grid Slot Object Reference)
   > SetHorizontalAlignment(self:Uniform Grid Slot Object Reference=@n4, InHorizontalAlignment=HAlign_Fill)
   n5 = Get GridColumn() -> (GridColumn:int=0)
   > SetColumn!(execute, self:Uniform Grid Slot Object Reference=@n4, InColumn:int=@n5) -> (then)
   > SetRow!(execute, self:Uniform Grid Slot Object Reference=@n4, InRow:int=Get GridRow() -> (GridRow:int=0)) -> (then)
   > Increment Int(Value=Get GridColumn() -> (GridColumn=0))
   > Branch(Condition="Equal (Integer)"(A=@n5, B=2))
   > Increment Int(Value=Get GridRow() -> (GridRow=0))
   > Set GridColumn(GridColumn=0) -> (Output_Get=0)
   Cast To EUW_Foley_FoleyTypeCollectionWidget(Object=@n3)
  ]}
 UpdateSoundLibraryWidgets {nodes [
   Clear(TargetArray:Array of EUW Foley Sound Data Entry Object References=Get SoundDataEntryArray() -> (SoundDataEntryArray:Array of EUW Foley Sound Data Entry Object References))
   n1 = Get Sounds(self=Get SoundLibrary!() -> (SoundLibrary))
   > n2 = Keys!(execute, TargetMap:Map of Gameplay Tag Structures to Sound Base Soft Object References=@n1) -> (then, Keys:Array of Gameplay Tag Structures)
   > n3 = For Each Loop(Array:Array of Gameplay Tag Structures=@n2) -> (Array Element:Gameplay Tag Structure)
   > n4 = Create Widget!(
    execute:Exec
    WorldContextObject:Object Reference
    Class:User Widget Class Reference=&/EUW_Foley_SoundDataEntry.EUW_Foley_SoundDataEntry_C
    OwningPlayer:Player Controller Object Reference
    CurrentTag:Sound Library Tag Structure=MakeSoundLibraryTag(Tag=@n3)
    Sound:Sound Base Soft Object Reference=Find(TargetMap:Map of Gameplay Tag Structures to Sound Base Soft Object References=@n1, Key:"Gameplay Tag Structure (by ref)"=@n3) -> (Value:Sound Base Soft Object Reference)
    SoundLibrary:Sound Library Object Reference=Get SoundLibrary()
   ) -> (then:Exec, ReturnValue:EUW Foley Sound Data Entry Object Reference)
   > Add(
    TargetArray:Array of EUW Foley Sound Data Entry Object References=Get SoundDataEntryArray()
    NewItem:"EUW Foley Sound Data Entry Object Reference (by ref)"=@n4
   ) {is=title K2Node_CallArrayFunction}
   > AddChild(self=Get TestScrollBox() -> (TestScrollBox:Editor Utility Scroll Box Object Reference), Content=@n4)
   UpdateSoundLibraryWidgets() {is=title K2Node_FunctionEntry}
  ]}
]
variables [
 Array Element {type=string, default=""}
 Array Element_0 {type=string, default=""}
 DefaultFoleyTypes {type="array<string>", default [Barefoot, BarefootHeavy, BootsCombat, DressShoes, BootsHeeled, BootsMetal, Sandals]}
 GridColumn {type=integer}
 GridRow {type=integer}
 GroupedObjects {type="array<Object>", default []}
 HasNonWave {type=boolean}
 LibrarySounds {default {}}
 SelectedObject {type="array<Object>", default []}
 SoundDataEntryArray {type="array<EUW_Foley_SoundDataEntry_C>", default []}
 SoundLibrary {type=SoundLibrary, default=null}
 StanceType {type="array<string>", default [Walk,Run,Sprint,CrouchWalk,CrouchRun,SlideLong,SlideShort,Stop,Turn,Jump,Land,Scuff,Plant,LandHard,Step]}
 SurfaceTypes {
  type="array<string>"
  default [Asphalt, Dirt, Grass, Metal, Sand, Snow, Ice, Stone, Water, Wood, Chrome, ChromePetrified, Mud, WetConcrete, TreeCanopy, Carpet]
 }
]
widgets {
 info {rootWidgetClass=&GridPanel, widgetCount=19}
 widgets [
  GridPanel_0 {widgetClassPath=&GridPanel, widget=GridPanel_0}
  VerticalBox_0 {
   widgetClassPath=&VerticalBox
   parent=GridPanel_0
   widget=VerticalBox_0
   slot {class=GridSlot, properties {horizontalAlignment=HAlign_Left, verticalAlignment=VAlign_Bottom, padding=<ltrb 5 10 5 5>}}
  }
  HorizontalBox_0 {
   widgetClassPath=&HorizontalBox
   parent=VerticalBox_0
   widget=HorizontalBox_0
   slot {class=VerticalBoxSlot, properties {horizontalAlignment=HAlign_Left}}
  }
  DefaultFoleyTypeButton {widgetClassPath=&EditorUtilityButton, parent=HorizontalBox_0, widget=DefaultFoleyTypeButton, slot {class=HorizontalBoxSlot}}
  UEFN_TextBlock_C_2 {
   widgetClassPath=&UEFN_TextBlock
   parent=DefaultFoleyTypeButton
   widget=UEFN_TextBlock_C_2
   properties {
    justification=Center
    text=AddDefaultTypes
    font {
     bForceMonospaced=false
     bMaterialIsStencil=false
     fontMaterial=null
     fontObject=/LauncherSocial/BrutalType
     letterSpacing=0
     monospacedWidth=1
     size=9
     skewAmount=0
     typefaceFontName=Regular
     outlineSettings {
      bApplyOutlineToDropShadows=false
      bMiteredCorners=false
      bSeparateFillAlpha=false
      outlineMaterial=null
      outlineSize=0
      outlineColor=<rgba 0 0 0 1>
     }
    }
    shadowColorAndOpacity=<rgba 0 0 0 1>
   }
   slot {class=ButtonSlot}
  }
  BuildTemplatesButton {
   widgetClassPath=&EditorUtilityButton
   parent=HorizontalBox_0
   widget=BuildTemplatesButton
   slot {class=HorizontalBoxSlot, properties {horizontalAlignment=HAlign_Left}}
  }
  UEFN_TextBlock_C {
   widgetClassPath=&UEFN_TextBlock
   parent=BuildTemplatesButton
   widget=UEFN_TextBlock_C
   properties {
    justification=Center
    text=Build Templates
    font {
     bForceMonospaced=false
     bMaterialIsStencil=false
     fontMaterial=null
     fontObject=&BrutalType
     letterSpacing=0
     monospacedWidth=1
     size=9
     skewAmount=0
     typefaceFontName=Regular
     outlineSettings {
      bApplyOutlineToDropShadows=false
      bMiteredCorners=false
      bSeparateFillAlpha=false
      outlineMaterial=null
      outlineSize=0
      outlineColor=<rgba 0 0 0 1>
     }
    }
    shadowColorAndOpacity=<rgba 0 0 0 1>
   }
   slot {class=ButtonSlot}
  }
  AddToLibButton {
   widgetClassPath=&EditorUtilityButton
   parent=HorizontalBox_0
   widget=AddToLibButton
   slot {class=HorizontalBoxSlot, properties {horizontalAlignment=HAlign_Left}}
  }
  UEFN_TextBlock_C_1 {
   widgetClassPath=&UEFN_TextBlock
   parent=AddToLibButton
   widget=UEFN_TextBlock_C_1
   properties {
    justification=Center
    text=Add to Lib
    font {
     bForceMonospaced=false
     bMaterialIsStencil=false
     fontMaterial=null
     fontObject=&BrutalType
     letterSpacing=0
     monospacedWidth=1
     size=9
     skewAmount=0
     typefaceFontName=Regular
     outlineSettings {
      bApplyOutlineToDropShadows=false
      bMiteredCorners=false
      bSeparateFillAlpha=false
      outlineMaterial=null
      outlineSize=0
      outlineColor=<rgba 0 0 0 1>
     }
    }
    shadowColorAndOpacity=<rgba 0 0 0 1>
   }
   slot {class=ButtonSlot}
  }
  HorizontalBox_1 {
   widgetClassPath=&HorizontalBox
   parent=VerticalBox_0
   widget=HorizontalBox_1
   slot {class=VerticalBoxSlot, properties {padding=<ltrb 5 5 5 5>}}
  }
  AddFoleyTypeButton {widgetClassPath=&EditorUtilityButton, parent=HorizontalBox_1, widget=AddFoleyTypeButton, slot {class=HorizontalBoxSlot}}
  UEFN_TextBlock_C_3 {
   widgetClassPath=&UEFN_TextBlock
   parent=AddFoleyTypeButton
   widget=UEFN_TextBlock_C_3
   properties {
    justification=Center
    text=Add
    font {
     bForceMonospaced=false
     bMaterialIsStencil=false
     fontMaterial=null
     fontObject=&BrutalType
     letterSpacing=0
     monospacedWidth=1
     size=9
     skewAmount=0
     typefaceFontName=Regular
     outlineSettings {
      bApplyOutlineToDropShadows=false
      bMiteredCorners=false
      bSeparateFillAlpha=false
      outlineMaterial=null
      outlineSize=0
      outlineColor=<rgba 0 0 0 1>
     }
    }
    shadowColorAndOpacity=<rgba 0 0 0 1>
   }
   slot {class=ButtonSlot}
  }
  SizeBox_1 {
   widgetClassPath=&SizeBox
   parent=HorizontalBox_1
   widget=SizeBox_1
   properties {bOverride_WidthOverride=true, widthOverride=350}
   slot {class=HorizontalBoxSlot}
  }
  FoleyTextEntry {
   widgetClassPath=&EditorUtilityEditableTextBox
   bIsVariable=true
   parent=SizeBox_1
   widget=FoleyTextEntry
   slot {class=SizeBoxSlot}
  }
  SizeBox {
   widgetClassPath=&SizeBox
   parent=VerticalBox_0
   widget=SizeBox
   properties {bOverride_HeightOverride=true, heightOverride=100, maxDesiredWidth=460, minDesiredWidth=450, widthOverride=450}
   slot {class=VerticalBoxSlot, properties {padding=<ltrb 5 5 5 5>}}
  }
  FoleyTypeScrollBox {
   widgetClassPath=&EditorUtilityScrollBox
   bIsVariable=true
   parent=SizeBox
   widget=FoleyTypeScrollBox
   slot {class=SizeBoxSlot}
  }
  SizeBox_2 {
   widgetClassPath=&SizeBox
   parent=VerticalBox_0
   widget=SizeBox_2
   properties {bOverride_HeightOverride=true, heightOverride=500, maxDesiredHeight=1000, minDesiredHeight=500}
   slot {class=VerticalBoxSlot}
  }
  EditorUtilityScrollBox_0 {widgetClassPath=&EditorUtilityScrollBox, parent=SizeBox_2, widget=EditorUtilityScrollBox_0, slot {class=SizeBoxSlot}}
  ListGridPanel {
   widgetClassPath=/Script/UMG.UniformGridPanel
   bIsVariable=true
   parent=EditorUtilityScrollBox_0
   widget=ListGridPanel
   slot {class=ScrollBoxSlot, properties {size {sizeRule=Fill, value=1}}}
  }
 ]
}
